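; Gameplay ability for the Dismiss Drone quest in the Season 23 Battle Pass narrative, Part 01, reporting quest progress for the owning player.
; On activation from a gameplay event, it runs the parent logic, then sends a complex custom stat event with a count of one to the player's quest manager, using the quest tag variable.
; Ending the ability only defers to the parent class.
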(event Ability|EventActivateAbilityFromEvent
  (params (EventData :type Gameplay%20Event%20Data%20Structure%20%28by%20ref%29))
  ;[1 1 1] Activate ability
  (|CallParentFunction EventData)
  (Quest|SendComplexCustomStatEvent (Variables|GAVars|Data|GetOwningPlayerQuestManager) 0 (Variables|GAVars|Setup|GetT_Quest) "(GameplayTags=)" 1))

(event Ability|EventOnEndAbility
  (params (bWasCancelled :type Boolean))
  ;[1 1 1] On End Ability
  (|CallParentFunction bWasCancelled))
